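; Animation modifier that automatically places left and right foot sync markers on an animation sequence, on configurable notify tracks.
; On apply, it temporarily clears forced root lock, optionally removes existing markers, and samples raw world-space poses at a fixed step.
; It dots each foot's offset from a trajectory reference bone with root motion, marking where that flips from positive to negative.
; Revert removes both notify tracks.
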(event EventOnApply
  (params (AnimationSequence :type Anim%20Sequence%20Object%20Reference))
  (Variables|Default|SetIsRootMotionLockForced (AnimationBlueprintLibrary|RootMotion|IsRootMotionLockForced AnimationSequence))
  (AnimationBlueprintLibrary|RootMotion|SetIsRootMotionLockForced AnimationSequence)
  (if (Variables|Default|GetRemovePreExistingSyncMarkers)
    (AnimationBlueprintLibrary|MarkerSyncing|RemoveAllAnimationSyncMarkers AnimationSequence)
    (else
      _))
  (AnimationBlueprintLibrary|AnimationNotifies|AddAnimationNotifyTrack AnimationSequence (Variables|Default|GetLeftFootNotifyTrackName))
  (if (!= (Variables|Default|GetLeftFootNotifyTrackName) (Variables|Default|GetRightFootNotifyTrackName))
    (AnimationBlueprintLibrary|AnimationNotifies|AddAnimationNotifyTrack AnimationSequence (Variables|Default|GetRightFootNotifyTrackName))
    (else
      _))
  (for _index (range (+ (Math|Float|Truncate (/ (AnimationBlueprintLibrary|Animation|GetSequenceLength AnimationSequence) (Variables|Default|GetSampleStep))) 1))
    (CallFunction|DetectMarkerforTime (* _index (Variables|Default|GetSampleStep)) AnimationSequence))
  (AnimationBlueprintLibrary|RootMotion|SetIsRootMotionLockForced AnimationSequence (Variables|Default|GetIsRootMotionLockForced)))

(event EventOnRevert
  (params (AnimationSequence :type Anim%20Sequence%20Object%20Reference))
  (AnimationBlueprintLibrary|AnimationNotifies|RemoveAnimationNotifyTrack AnimationSequence (Variables|Default|GetLeftFootNotifyTrackName))
  (AnimationBlueprintLibrary|AnimationNotifies|RemoveAnimationNotifyTrack AnimationSequence (Variables|Default|GetRightFootNotifyTrackName)))

(fn DetectMarkerForTime
  (params (Time :type Float%20%28double-precision%29) (AnimSequence :type Anim%20Sequence%20Base%20Object%20Reference))
  (LocalVariables|SetLocalTime Time)
  (bind _asanim_sequence (Utilities|Casting|CastToAnimSequence AnimSequence))
  (LocalVariables|SetLocalAnimSequence _asanim_sequence)
  (Utilities|FlowControl|Sequence
    (:then_0
      ;[1 1 1] Get root motion translation
      (LocalVariables|SetLocalRootMotionTranslation (CallFunction|ExtractRootMotion (LocalVariables|GetLocalAnimSequence) (LocalVariables|GetLocalTime) (+ (LocalVariables|GetLocalTime) (Variables|Default|GetSampleStep)))))
    (:then_1
      ;[1 1 1] Get trajectory reference bone position
      (LocalVariables|SetLocalTrajRefBonePos (CallFunction|GetBonePoseHelper (LocalVariables|GetLocalAnimSequence) (Variables|Default|GetTrajectoryReferenceBoneName) (LocalVariables|GetLocalTime))))
    (:then_2
      ;[1 1 1] Get left foot bone position
      (LocalVariables|SetLocalLeftBonePos (CallFunction|GetBonePoseHelper (LocalVariables|GetLocalAnimSequence) (Variables|Default|GetLeftFootBoneName) (LocalVariables|GetLocalTime))))
    (:then_3
      ;[1 1 1] Right right foot bone position
      (LocalVariables|SetLocalRightBonePos (CallFunction|GetBonePoseHelper (LocalVariables|GetLocalAnimSequence) (Variables|Default|GetRightFootBoneName) (LocalVariables|GetLocalTime))))
    (:then_4
      ;[1 1 1] Place marker for left foot if desired
      (bind (_currentdotproduct _shouldplacemarker) (CallFunction|ShouldPlaceMarkerforFoot (LocalVariables|GetLocalRootMotionTranslation) (LocalVariables|GetLocalLeftBonePos) (LocalVariables|GetLocalTrajRefBonePos) (Variables|Default|GetLastLeftFootDotProduct)))
      (Variables|Default|SetLastLeftFootDotProduct _currentdotproduct)
      (if _shouldplacemarker
        (AnimationBlueprintLibrary|MarkerSyncing|AddAnimationSyncMarker (LocalVariables|GetLocalAnimSequence) (Variables|Default|GetLeftFootMarkerName) (LocalVariables|GetLocalTime) (Variables|Default|GetLeftFootNotifyTrackName))))
    (:then_5
      ;[1 1 1] Place marker for right foot if desired
      (bind (_currentdotproduct_2 _shouldplacemarker_2) (CallFunction|ShouldPlaceMarkerforFoot (LocalVariables|GetLocalRootMotionTranslation) (LocalVariables|GetLocalRightBonePos) (LocalVariables|GetLocalTrajRefBonePos) (Variables|Default|GetLastRightFootDotProduct)))
      (Variables|Default|SetLastRightFootDotProduct _currentdotproduct_2)
      (if _shouldplacemarker_2
        (AnimationBlueprintLibrary|MarkerSyncing|AddAnimationSyncMarker (LocalVariables|GetLocalAnimSequence) (Variables|Default|GetRightFootMarkerName) (LocalVariables|GetLocalTime) (Variables|Default|GetRightFootNotifyTrackName))))))

(fn ExtractRootMotion
  (params (Animation :type Anim%20Sequence%20Object%20Reference) (StartTime :type Float%20%28double-precision%29) (EndTime :type Float%20%28double-precision%29))
  (return (- (CallFunction|GetBonePoseHelper Animation (Variables|Default|GetRootBoneName) (Math|Float|Clamp(Float) EndTime 0.0 (Animation|GetPlayLength Animation))) (CallFunction|GetBonePoseHelper Animation (Variables|Default|GetRootBoneName) StartTime))))

(fn GetBonePoseHelper
  (params (AnimSequence :type Anim%20Sequence%20Base%20Object%20Reference) (BoneName :type Name) (Time :type Float%20%28double-precision%29))
  (bind _pose (Animation|Pose|GetAnimPoseatTime AnimSequence Time (Utilities|Struct|MakeAnimPoseEvaluationOptions "Raw")))
  (bind (_returnvalue_location _returnvalue_rotation) (Math|Transform|BreakTransform (Animation|Pose|GetBonePose _pose BoneName "World")))
  (return _returnvalue_location _returnvalue_rotation))

(fn ShouldPlaceMarkerForFoot
  (params (RootMotionTranslation :type Vector) (FootPos :type Vector) (TrajectoryPos :type Vector) (LastDotProduct :type Float%20%28double-precision%29))
  ;[1 1 1] Test if the foot was ahead of the trajectory reference bone (in direction of root motion) but is now behind.
  (return (Math|Vector|DotProduct RootMotionTranslation (- FootPos TrajectoryPos)) (and (< (Math|Vector|DotProduct RootMotionTranslation (- FootPos TrajectoryPos)) 0) (> LastDotProduct 0))))
