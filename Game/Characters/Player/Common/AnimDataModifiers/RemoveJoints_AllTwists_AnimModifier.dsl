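; Animation modifier that strips bone animation for twist joints from a player Animation Sequence when applied.
; A three-branch Sequence calls RemoveBoneAnimation, children included, on both left and right calf, upperarm and lowerarm twist bones.
; Finalize is deferred until the last removal, so the track edits are finalized once.
; Despite the AllTwists name, no thigh twist bones are removed.
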(event EventOnApply
  (params (AnimationSequence :type Anim%20Sequence%20Object%20Reference))
  (Utilities|FlowControl|Sequence
    (:then_0
      (AnimationBlueprintLibrary|Bones|RemoveBoneAnimation AnimationSequence "calf_twist_01_l" true false)
      (AnimationBlueprintLibrary|Bones|RemoveBoneAnimation AnimationSequence "calf_twist_02_l" true false)
      (AnimationBlueprintLibrary|Bones|RemoveBoneAnimation AnimationSequence "calf_twist_01_r" true false)
      (AnimationBlueprintLibrary|Bones|RemoveBoneAnimation AnimationSequence "calf_twist_02_r" true false))
    (:then_1
      (AnimationBlueprintLibrary|Bones|RemoveBoneAnimation AnimationSequence "upperarm_twist_01_l" true false)
      (AnimationBlueprintLibrary|Bones|RemoveBoneAnimation AnimationSequence "upperarm_twist_02_l" true false)
      (AnimationBlueprintLibrary|Bones|RemoveBoneAnimation AnimationSequence "upperarm_twist_01_r" true false)
      (AnimationBlueprintLibrary|Bones|RemoveBoneAnimation AnimationSequence "upperarm_twist_02_r" true false))
    (:then_2
      (AnimationBlueprintLibrary|Bones|RemoveBoneAnimation AnimationSequence "lowerarm_twist_01_l" true false)
      (AnimationBlueprintLibrary|Bones|RemoveBoneAnimation AnimationSequence "lowerarm_twist_02_l" true false)
      (AnimationBlueprintLibrary|Bones|RemoveBoneAnimation AnimationSequence "lowerarm_twist_01_r" true false)
      (AnimationBlueprintLibrary|Bones|RemoveBoneAnimation AnimationSequence "lowerarm_twist_02_r"))))
