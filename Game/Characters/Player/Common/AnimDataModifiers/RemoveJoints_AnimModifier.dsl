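; Animation modifier for player animations that strips animation tracks from a fixed set of auxiliary joints.
; On apply, it calls RemoveBoneAnimation on the animation sequence once for each hardcoded bone name.
; The targeted joints are the backpack, cape and fp attach bones, three warp points and the camera root.
; Only the apply event is defined, with no revert logic shown.
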(event EventOnApply
  (params (AnimationSequence :type Anim%20Sequence%20Object%20Reference))
  (AnimationBlueprintLibrary|Bones|RemoveBoneAnimation AnimationSequence "attach_backpack")
  (AnimationBlueprintLibrary|Bones|RemoveBoneAnimation AnimationSequence "attach_cape")
  (AnimationBlueprintLibrary|Bones|RemoveBoneAnimation AnimationSequence "warp_point_01")
  (AnimationBlueprintLibrary|Bones|RemoveBoneAnimation AnimationSequence "warp_point_02")
  (AnimationBlueprintLibrary|Bones|RemoveBoneAnimation AnimationSequence "warp_point_03")
  (AnimationBlueprintLibrary|Bones|RemoveBoneAnimation AnimationSequence "camera_root")
  (AnimationBlueprintLibrary|Bones|RemoveBoneAnimation AnimationSequence "attach_fp"))
