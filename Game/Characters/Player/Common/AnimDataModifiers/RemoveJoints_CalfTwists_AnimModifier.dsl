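; Animation modifier for player animation sequences that strips keyed animation from the calf twist joints.
; On apply, it calls RemoveBoneAnimation in sequence for both left and right calf twist bones, 01 and 02 on each side.
; It defines no revert logic, only the apply event.
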(event EventOnApply
  (params (AnimationSequence :type Anim%20Sequence%20Object%20Reference))
  (AnimationBlueprintLibrary|Bones|RemoveBoneAnimation AnimationSequence "calf_twist_01_l")
  (AnimationBlueprintLibrary|Bones|RemoveBoneAnimation AnimationSequence "calf_twist_02_l")
  (AnimationBlueprintLibrary|Bones|RemoveBoneAnimation AnimationSequence "calf_twist_01_r")
  (AnimationBlueprintLibrary|Bones|RemoveBoneAnimation AnimationSequence "calf_twist_02_r"))
