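# Parent material for transparent player characters, set to unlit translucent shading.
# Usage flags cover skeletal meshes, morph targets, mesh deformers and Nanite.
# The graph holds only a constant and a multiply that negates world position, with no material outputs shown connected and no parameters exposed.
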Material /Game/Characters/Player/Common/TransparentCharacterParent
graph {
 properties {
  bUsedWithMeshDeformer=true
  bUsedWithMorphTargets=true
  bUsedWithNanite=true
  bUsedWithSkeletalMesh=true
  blendMode=BLEND_Translucent
  pixelDepthOffsetMode=PDOM_Legacy
  shadingModel=MSM_Unlit
  usedShadingModels=Unlit
 }
 nodes [
  Constant()
  Multiply(B=WorldPosition().XYZ) {constA=-1, constB=-1}
 ]
}
